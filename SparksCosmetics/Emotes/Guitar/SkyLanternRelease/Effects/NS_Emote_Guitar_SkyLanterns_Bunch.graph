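# Niagara system for the Sky Lantern Release guitar emote: a bunch of glowing lanterns that rise into a spinning, bobbing ring.
# Lantern paths are computed by blending start, mid and ring positions over particle and system age, with ID-based phase and lifetimes synced to end together.
# Lanterns feed mesh, sprite and light renderers; a front-end preview flag dims some colors.
# Sparkles read lanterns through an attribute reader, inheriting position and varied color, then fall.
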NiagaraSystem /SparksCosmetics/Emotes/Guitar/SkyLanternRelease/Effects/NS_Emote_Guitar_SkyLanterns_Bunch
dependencies {
 usedDataInterfaces [
  /Script/Niagara.NiagaraDataInterfaceSimpleCounter
  &NiagaraDataInterfaceCurve
  &NiagaraDataInterfaceColorCurve
  &NiagaraDataInterfaceParticleRead
 ]
 usedDynamicInputs []
 usedModules [&SystemState]
}
emitters {
 Lanterns {
  emitterUpdateScript [
   &EmitterState(Inactive Response=::NewEnumerator1, Life Cycle Mode=::NewEnumerator1, Loop Duration=5.8667, UseLoopDelay=false)
   &SpawnRate(
    Spawn Probability=&ClampFloat(Float=&Add_Float(A=$Engine.System.Age, B=-4))
    SpawnRate=3
    Use Spawn Probability=true
   )
   &SpawnBurst_Instantaneous(Spawn Count=7, Spawn Time=5.8667, Use Spawn Probability=false, UseLoopCountLimit=false) {module=SpawnBurst_Instantaneous001}
  ]
  particleSpawnScript [
   /SetVariables(Particles.IDNum:NiagaraInt32=/Niagara/DynamicInputs/DataInterface/GenerateIntFromCounter(SimpleCounter:NiagaraDataInterfaceSimpleCounter=DataInterface() {propertyValues {GpuSyncMode=None, InitialValue=0}})) {isSetParameters=true}
   /SetVariables(Particles.NormalizedID:NiagaraFloat=&Multiply_FloatByInt(Float=0.1429, Integer=$Particles.IDNum)) {isSetParameters=true}
   &InitializeParticle(
    Color Maximum=<rgba 0.971 0.6622 0.2583 1>
    Color Minimum=<rgba 0.8 0.4318 0.1096 1>
    Color Mode=::NewEnumerator2
    Lifetime=&Subtract_Float(A=&RandomRangeFloat(Minimum=14.8, Maximum=15.1), B=$Engine.System.Age)
    Mesh Scale Mode=::NewEnumerator3
    Mesh Uniform Scale=0.35
    Position Mode=::NewEnumerator1
    Sprite Size Mode=::NewEnumerator3
    Uniform Sprite Size=75
    UsePositionOffset=false
    Write Material Random=true
   )
   /SetVariables(Particles.MeshIndex=&RandomRangeInt(Minimum=0, Maximum=2)) {isSetParameters=true}
   /SetVariables(Particles.MeshIndex=&ModuloInt(A=$Particles.IDNum, B=3)) {enabled=false, isSetParameters=true}
   &KillParticles(Kill Particles=&SetBoolByIntComparison(A=$Particles.IDNum, B=6, Comparison Type=::NewEnumerator0))
  ]
  particleUpdateScript [
   /SetVariables(Particles.SpinTime:NiagaraFloat=&Multiply_Float(
     A=&Add_Float(A=$Emitter.Age, B=-1.5)
     B=&RandomRangeFloat(Minimum=0.132, Maximum=0.135)
    )) {isSetParameters=true}
   &ParticleState()
   &ScaleMeshSize(Scale Factor=&VectorFromFloat(Value=&FloatFromCurve(FloatCurve=DataInterface() {propertyValues {Curve {keys [
          {arriveTangent=0.0315, interpMode=RCIM_Cubic, leaveTangent=0.0315, tangentMode=RCTM_User, time=0, value=0}
          {
           arriveTangent=3.7421
           arriveTangentWeight=0.3859
           interpMode=RCIM_Cubic
           leaveTangent=3.7421
           leaveTangentWeight=0.0155
           tangentMode=RCTM_User
           time=0.03
           value=1
          }
          {arriveTangent=1.1121, tangentMode=RCTM_User, time=0.05, value=1}
          <at 0.95 1>
          <at 0.97 1>
          <at 1 0>
         ]}}})))
   &ScaleSpriteSize(Uniform Curve Sprite Scale=DataInterface() {propertyValues {Curve {keys [
        {arriveTangent=0.0315, interpMode=RCIM_Cubic, leaveTangent=0.0315, tangentMode=RCTM_User, time=0, value=0}
        {
         arriveTangent=3.7421
         arriveTangentWeight=0.3859
         interpMode=RCIM_Cubic
         leaveTangent=3.7421
         leaveTangentWeight=0.0155
         tangentMode=RCTM_User
         time=0.03
         value=1
        }
        {arriveTangent=1.1121, tangentMode=RCTM_User, time=0.05, value=1}
        <at 0.95 1>
        <at 0.97 1>
        <at 1 0>
       ]}}})
   &SolveForcesAndVelocity(Clamp Velocity=false, Limit Acceleration=false)
   /SetVariables(Particles.SpinPosition:NiagaraPosition=&AddVectorToPosition(
     Position=&RandomRangeVector(Minimum=<0 0 225>, Maximum=<0 0 250>)
     Vector=&Multiply_VectorByFloat(
      Vector=&MakeVector(
       X=&Sine(Normalized Angle=&Add_Float(A=$Particles.NormalizedID, B=$Particles.SpinTime), Period=1, Scale=1, Bias=0)
       Y=&Cosine(Normalized Angle=&Add_Float(A=$Particles.NormalizedID, B=$Particles.SpinTime), Period=1, Scale=1, Bias=0)
       Z=0
      )
      Float=&RandomRangeFloat(Minimum=65, Maximum=75)
     )
    )) {isSetParameters=true}
   /SetVariables(
    Particles.BobOffset:NiagaraFloat=&Waveform(
     Global Phase=&Add_Float(
      A=&Multiply_Float(A=$Emitter.Age, B=&RandomRangeFloat(Minimum=0.34, Maximum=0.36))
      B=&Add_Float(A=$Particles.NormalizedID, B=0.25)
     )
     Global Amplitude Scale=&RandomRangeFloat(Minimum=10, Maximum=20)
     "Waveform [1]"=::NewEnumerator0
     "[1] Frequency"=1
     "[1] Amplitude Min/Max"=<-1 1>
    )
    Particles.StartPos:Vector3f=&Add_Vector(
     A=&RandomRangeVector(Minimum=<-25 -25 50>, Maximum=<25 25 100>)
     B=&Multiply_Vector(A=<1 1 0>, B=Linked() {linkedVariable {name=Particles.SpinPosition, type=NiagaraPosition}})
    )
   ) {isSetParameters=true}
   /SetVariables(Particles.StartPos:Vector3f=&Add_Vector(
     A=&RandomRangeVector(Minimum=<0 0 40>, Maximum=<1 1 60>)
     B=&self:MakeRing(Default:Vector3f=<0 0 0>, Circle Radius:NiagaraFloat=&RandomRangeFloat(Minimum=100, Maximum=150))
    )) {enabled=false, isSetParameters=true}
   /SetVariables(Particles.MidPos:Vector3f=&Add_Vector(
     A=&RandomRangeVector(Minimum=<0 0 125>, Maximum=<0 0 175>)
     B=&Multiply_Vector(A=$Particles.StartPos, B=&RandomRangeVector(Minimum=<1.5 1.5 0>, Maximum=<1.75 1.75 0>))
    )) {isSetParameters=true}
   /SetVariables(Particles.MidPos:Vector3f=&Add_Vector(
     A=&RandomRangeVector(Minimum=<-75 -75 125>, Maximum=<75 75 175>)
     B=&Multiply_Vector(A=$Particles.StartPos, B=&RandomRangeVector(Minimum=<0.9 0.9 0>, Maximum=<1.25 1.25 0>))
    )) {enabled=false, isSetParameters=true}
   /SetVariables(Particles.PositionMod:NiagaraPosition=&AddVectorToPosition(
     Position=&LerpPosition(
      A=Linked() {linkedVariable {name=Particles.StartPos, type=Vector3f}}
      B=&LerpPosition(
       A=Linked() {linkedVariable {name=Particles.MidPos, type=Vector3f}}
       B=$Particles.SpinPosition
       Alpha=&FloatFromCurve(
        FloatCurve=DataInterface() {propertyValues {Curve {keys [
            <at 6 0 RCIM_Cubic>
            {arriveTangent=0.8667, interpMode=RCIM_Cubic, leaveTangent=0.8667, tangentMode=RCTM_User, time=7.25, value=0.5}
            {arriveTangentWeight=0.1668, interpMode=RCIM_Cubic, leaveTangent=0.2168, leaveTangentWeight=0.1668, time=8, value=1}
           ]}}}
        CurveIndex=$Engine.System.Age
       )
      )
      Alpha=&FloatFromCurve(
       FloatCurve=DataInterface() {propertyValues {Curve {keys [
           {arriveTangent=-0.6124, interpMode=RCIM_Cubic, leaveTangent=-0.6124, tangentMode=RCTM_User, time=0.1, value=0}
           {arriveTangent=0.5789, interpMode=RCIM_Cubic, leaveTangent=0.5789, time=0.5, value=0.05}
           {arriveTangentWeight=0.1735, interpMode=RCIM_Cubic, leaveTangentWeight=0.347, tangentMode=RCTM_User, time=2, value=1.1}
           <at 3 1 RCIM_Cubic>
          ]}}}
       CurveIndex=$Particles.Age
      )
     )
     Vector=&Multiply_VectorByFloat(Vector=<0 0 1>, Float=$Particles.BobOffset)
    )) {isSetParameters=true}
   &UpdateMeshOrientation(
    Facing Direction=&Multiply_Vector(A=Linked() {linkedVariable {name=Particles.PositionMod, type=NiagaraPosition}}, B=<1 1 0>)
    Facing Interpolation=false
    Orientation Method=::NewEnumerator1
    Override Quat=false
    TransformFacing=false
    TransformUp=false
   )
   /SetVariables(
    Particles.LightColor=&Multiply_LinearColorByFloat(Linear Color=$Particles.Color, Float=10)
    Particles.LightPosition:NiagaraPosition=&AddVectorToPosition(Position=$Particles.PositionMod, Vector=<0 0 -18>)
    Particles.LightRadius=&Multiply_Float(A=150, B=&MakeFloatFromVector(Vector=$Particles.Scale, Channel=::NewEnumerator0))
   ) {isSetParameters=true}
   /SetVariables(
    Particles.ButterflyColor:LinearColor=&MakeCustomLinearColorFromBool(Bool=$User.bIsFrontEndPreview, True Color=<rgba 0.474 0.474 0.474 1>, False Color=<rgba 1 1 1 1>)
    Particles.ButterflyScale:Vector3f=&Multiply_VectorByFloat(Vector=$Particles.Scale, Float=&RandomRangeFloat(Minimum=0.1, Maximum=0.13))
    Particles.GlowColor=&MakeLinearColorFromVectorAndFloat(
     "Vector (RGB)"=&Add_Vector(A=&RandomVector(Vector Scale=0.2), B=<0.5 0.5 1>)
     "Float (Alpha)"=&MakeCustomFloatFromBool(Bool=$User.bIsFrontEndPreview, True Float=0.75, False Float=1)
    )
   ) {isSetParameters=true}
   &DynamicMaterialParameters(
    Index 0 Param 1=0.25
    Index 0 Param 3=1
    Index 0 Param 4=&Multiply_Float(A=$Engine.Time, B=&RandomRangeFloat(Minimum=0.25, Maximum=0.5))
    Index 1 Param 1=3000
    Index 1 Param 2=1
    Module.Param0WriteEnabled=true
    Module.Param0WriteEnabled1=true
    Module.Param1WriteEnabled=true
    Module.Param1WriteEnabled1=true
    Module.Param2WriteEnabled=true
    Module.Param2WriteEnabled1=true
    Module.Param3WriteEnabled=true
    Module.Param3WriteEnabled1=true
    Write Parameter Index 1=true
   )
   &Color(Color=&ColorFromCurve(
     ColorCurve=DataInterface() {propertyValues {
       AlphaCurve {keys [<at 0 1 RCIM_Constant>]}
       BlueCurve {keys [<at 0 0 RCIM_Constant>,<at 0.263 0 RCIM_Constant>,<at 0.3947 0 RCIM_Constant>,<at 0.5592 1 RCIM_Constant>,<at 1 1 RCIM_Constant>]}
       GreenCurve {keys [<at 0 0 RCIM_Constant>,<at 0.263 1 RCIM_Constant>,<at 0.3947 1 RCIM_Constant>,<at 0.5592 0 RCIM_Constant>,<at 1 0 RCIM_Constant>]}
       RedCurve {keys [<at 0 1 RCIM_Constant>,<at 0.263 1 RCIM_Constant>,<at 0.3947 0 RCIM_Constant>,<at 0.5592 0 RCIM_Constant>,<at 1 1 RCIM_Constant>]}
      }}
     CurveIndex=$User.NormalizedNotifyProgress
    )) {enabled=false}
  ]
  rendererClasses [
   &NiagaraMeshRendererProperties
   &NiagaraLightRendererProperties
   &NiagaraSpriteRendererProperties
  ]
  renderers [
   {rendererClass=&NiagaraMeshRendererProperties, rendererIndex=0}
   {rendererClass=&NiagaraLightRendererProperties, rendererIndex=1}
   {rendererClass=&NiagaraSpriteRendererProperties, rendererIndex=2}
   {rendererClass=&NiagaraSpriteRendererProperties, rendererIndex=3}
   {rendererClass=&NiagaraMeshRendererProperties, rendererIndex=4}
  ]
 }
 Sparkles {
  emitterSpawnScript [/SetVariables(Emitter.Particle Attribute Reader=DataInterface() {propertyValues {EmitterBinding {bindingMode=Other, emitterName=Lanterns}}}) {isSetParameters=true}]
  emitterUpdateScript [
   &EmitterState()
   &SpawnRate(
    SpawnRate=&self:GetParentCount(
     Particle Attribute Reader:NiagaraDataInterfaceParticleRead=$Emitter.Particle Attribute Reader
     Multiplier:NiagaraFloat=&RandomRangeFloat(Minimum=5, Maximum=&Lerp_Float(A=10, B=25, Alpha=$System.NormalizedLoopAge), Evaluation Type=::NewEnumerator1)
     Addition:NiagaraFloat=0
    )
    Use Spawn Probability=false
   )
  ]
  particleSpawnScript [
   /SetVariables(Particles.PickParent:NiagaraInt32=&RandomRangeInt(
     Minimum=0
     Maximum=&MakeIntFromFloat(Float=&self:GetParentCount(
       Particle Attribute Reader:NiagaraDataInterfaceParticleRead=$Emitter.Particle Attribute Reader
       Multiplier:NiagaraFloat=1
       Addition:NiagaraFloat=-1
      ))
    )) {isSetParameters=true}
   &self:GetParentInfo(
    Hue Adjustment:NiagaraFloat=&RandomRangeFloat(Minimum=-0.1, Maximum=0.1)
    Normalize Color=true
    Parent Attribute Reader:NiagaraDataInterfaceParticleRead=$Emitter.Particle Attribute Reader
    ParentID:NiagaraInt32=$Particles.PickParent
    Saturation Adjustment:NiagaraFloat=&RandomRangeFloat(Minimum=0.5, Maximum=1.2)
    Value Adjustment:NiagaraFloat=&RandomRangeFloat(Minimum=0.3, Maximum=1.5)
   ) {module=GetParentInfo}
   &InitializeParticle(
    Color=$Particles.InheritedColor
    Lifetime=&Lerp_Float(
     A=0.4
     B=2.5
     Alpha=&FloatFromCurve(
      FloatCurve=DataInterface() {propertyValues {Curve {keys [
          {arriveTangent=0.6317, interpMode=RCIM_Cubic, leaveTangent=0.6317, tangentMode=RCTM_User, time=0, value=0}
          {arriveTangent=2.8093, interpMode=RCIM_Cubic, leaveTangent=2.8093, tangentMode=RCTM_User, time=1, value=1}
         ]}}}
      CurveIndex=&RandomRangeFloat(Minimum=0, Maximum=1)
     )
    )
    Position Offset=&AddVectorToPosition(Position=$Particles.ParentPosition, Vector=&RandomVector(Vector Scale=3))
    Position Offset Coordinate Space=::Local
    Sprite Rotation Mode=::NewEnumerator2
    Sprite Size Mode=::NewEnumerator1
    Uniform Sprite Size Max=20
    Uniform Sprite Size Min=10
    UsePositionOffset=true
    Write Material Random=true
   )
   &CameraOffset(Camera Offset Amount=$System.OffsetLanterns)
   &AddVelocity(Velocity=&RandomRangeVector(Minimum=<0 0 -5>, Maximum=<0 0 -75>))
  ]
  particleUpdateScript [
   &ParticleState()
   &SpriteRotationRate(Rotation Rate=&RandomRangeFloat(Minimum=-300, Maximum=300))
   &ScaleSpriteSize(
    Scale Sprite Size Mode=::NewEnumerator0
    Uniform Scale Factor=&FloatFromCurve(FloatCurve=DataInterface() {propertyValues {Curve {keys [
         {arriveTangent=3.1235, interpMode=RCIM_Cubic, leaveTangent=3.1235, tangentMode=RCTM_User, time=0, value=0}
         {interpMode=RCIM_Cubic, tangentMode=RCTM_User, time=0.5, value=1}
         {arriveTangent=-3.4019, tangentMode=RCTM_User, time=1, value=0}
        ]}}})
   )
   &Drag(Drag=&RandomRangeFloat(Minimum=1, Maximum=2), Use Linear Drag=true, Use Rotational Drag=false)
   &GravityForce(Gravity=&RandomRangeVector(Minimum=<0 0 -150>, Maximum=<0 0 -25>))
   &SolveForcesAndVelocity(Clamp Velocity=false, Limit Acceleration=false)
   &Color(Color=&MakeCustomLinearColorFromBool(Bool=$User.bIsFrontEndPreview, True Color=<rgba 1 0 0 1>, False Color=<rgba 0 0 1 1>)) {enabled=false}
   /SetVariables(Particles.GlowColor=&MultiplyLinearColor(Linear Color=$Particles.Color, Scale Factor=<0.3 0.3 0.75 1>)) {isSetParameters=true}
  ]
  rendererClasses [&NiagaraSpriteRendererProperties]
  renderers [
   {rendererClass=&NiagaraSpriteRendererProperties}
   {rendererClass=&NiagaraSpriteRendererProperties, rendererIndex=1}
  ]
 }
}
summary {
 systemName=NS_Emote_Guitar_SkyLanterns_Bunch
 emitters [
  {
   emitterName=Lanterns
   rendererClasses [
    &NiagaraMeshRendererProperties
    &NiagaraLightRendererProperties
    &NiagaraSpriteRendererProperties
   ]
  }
  {emitterName=Sparkles, rendererClasses [&NiagaraSpriteRendererProperties]}
 ]
 userVariables [
  User.bIsFrontEndPreview {type=NiagaraBool}
  User.NormalizedNotifyProgress {type=NiagaraFloat, defaultValue=0, description=""}
 ]
}
