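# Niagara solo effect for the Sky Lantern Release guitar emote, showing a glowing butterfly with falling sparkles.
# SoloButterfly makes a delayed burst with mesh, sprite and light renderers, scaling in and out as its light ramps with system age, dimmed in front-end preview.
# Sparkles spawn near the system origin, gated by system age, their maximum rate rising with NormalizedNotifyProgress.
# Their color reader targets a "Lanterns" emitter absent from this system.
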NiagaraSystem /SparksCosmetics/Emotes/Guitar/SkyLanternRelease/Effects/NS_Emote_Guitar_SkyLanterns_Solo
dependencies {
 usedDataInterfaces [&NiagaraDataInterfaceParticleRead, &NiagaraDataInterfaceCurve]
 usedDynamicInputs []
 usedModules [&SystemState]
}
emitters {
 SoloButterfly {
  emitterUpdateScript [
   &EmitterState()
   &SpawnBurst_Instantaneous(Spawn Time=1.48, Use Spawn Probability=false, UseLoopCountLimit=true)
  ]
  particleSpawnScript [&InitializeParticle(
    Color Mode=::NewEnumerator0
    Lifetime=15.1
    Mesh Scale Mode=::NewEnumerator3
    Mesh Uniform Scale=0.35
    Position=<0 0 22>
    Position Mode=::NewEnumerator1
    Sprite Size Mode=::NewEnumerator3
    Uniform Sprite Size=75
    UsePositionOffset=false
    Write Material Random=true
   )]
  particleUpdateScript [
   &ParticleState()
   &ScaleMeshSize(Scale Factor=&VectorFromFloat(Value=&FloatFromCurve(FloatCurve=DataInterface() {propertyValues {Curve {keys [
          {arriveTangent=0.0315, interpMode=RCIM_Cubic, leaveTangent=0.0315, tangentMode=RCTM_User, time=0, value=0}
          {
           arriveTangent=3.7421
           arriveTangentWeight=0.3859
           interpMode=RCIM_Cubic
           leaveTangent=3.7421
           leaveTangentWeight=0.0155
           tangentMode=RCTM_User
           time=0.03
           value=1
          }
          {arriveTangent=1.1121, tangentMode=RCTM_User, time=0.05, value=1}
          <at 0.95 1>
          <at 0.97 1>
          <at 1 0>
         ]}}})))
   &ScaleSpriteSize(Uniform Curve Sprite Scale=DataInterface() {propertyValues {Curve {keys [
        {arriveTangent=0.0315, interpMode=RCIM_Cubic, leaveTangent=0.0315, tangentMode=RCTM_User, time=0, value=0}
        {
         arriveTangent=3.7421
         arriveTangentWeight=0.3859
         interpMode=RCIM_Cubic
         leaveTangent=3.7421
         leaveTangentWeight=0.0155
         tangentMode=RCTM_User
         time=0.03
         value=1
        }
        {arriveTangent=1.1121, tangentMode=RCTM_User, time=0.05, value=1}
        <at 0.95 1>
        <at 0.97 1>
        <at 1 0>
       ]}}})
   &SolveForcesAndVelocity(Clamp Velocity=false, Limit Acceleration=false)
   &UpdateMeshOrientation(
    Facing Direction=&Multiply_Vector(A=Linked() {linkedVariable {name=Particles.Position, type=NiagaraPosition}}, B=<1 1 0>)
    Facing Interpolation=false
    Orientation Method=::NewEnumerator1
    Override Quat=false
    TransformFacing=false
    TransformUp=false
   )
   &DynamicMaterialParameters(
    Index 0 Param 1=0.25
    Index 0 Param 3=1
    Index 0 Param 4=&Multiply_Float(A=$Engine.Time, B=&RandomRangeFloat(Minimum=0.25, Maximum=0.5))
    Index 1 Param 1=3000
    Index 1 Param 2=1
    Module.Param0WriteEnabled=true
    Module.Param0WriteEnabled1=true
    Module.Param1WriteEnabled=true
    Module.Param1WriteEnabled1=true
    Module.Param2WriteEnabled=true
    Module.Param2WriteEnabled1=true
    Module.Param3WriteEnabled=true
    Module.Param3WriteEnabled1=true
    Write Parameter Index 1=true
   )
   /SetVariables(
    Particles.ButterflyColor:LinearColor=&MakeCustomLinearColorFromBool(Bool=$User.bIsFrontEndPreview, True Color=<rgba 0.474 0.474 0.474 1>, False Color=<rgba 1 1 1 1>)
    Particles.ButterflyScale:Vector3f=&Multiply_VectorByFloat(Vector=$Particles.Scale, Float=&RandomRangeFloat(Minimum=0.1, Maximum=0.13))
    Particles.GlowColor=&MakeLinearColorFromVectorAndFloat(
     "Vector (RGB)"=&Add_Vector(A=&RandomVector(Vector Scale=0.2), B=<0.5 0.5 1>)
     "Float (Alpha)"=&MakeCustomFloatFromBool(Bool=$User.bIsFrontEndPreview, True Float=0.75, False Float=1)
    )
   ) {isSetParameters=true}
   /SetVariables(
    Particles.LightColor=&Multiply_LinearColorByFloat(
     Linear Color=&RandomRangeLinearColor(Minimum=<rgba 0.5 0.2939 0.0685 1>, Maximum=<rgba 0.971 0.6305 0.2583 1>)
     Float=&ClampFloat(Float=&Multiply_Float(A=$System.Age, B=2), Max=10)
    )
    Particles.LightPosition:NiagaraPosition=<0 0 0>
    Particles.LightRadius=52.5
   ) {isSetParameters=true}
  ]
  rendererClasses [
   &NiagaraMeshRendererProperties
   &NiagaraSpriteRendererProperties
   &NiagaraLightRendererProperties
  ]
  renderers [
   {rendererClass=&NiagaraMeshRendererProperties, rendererIndex=0}
   {rendererClass=&NiagaraSpriteRendererProperties, rendererIndex=1}
   {rendererClass=&NiagaraLightRendererProperties, rendererIndex=2}
  ]
 }
 Sparkles {
  emitterSpawnScript [/SetVariables(Emitter.Particle Attribute Reader=DataInterface() {propertyValues {EmitterBinding {bindingMode=Other, emitterName=Lanterns}}}) {isSetParameters=true}]
  emitterUpdateScript [
   &EmitterState(Life Cycle Mode=::NewEnumerator1, Loop Duration=1, UseLoopDelay=false)
   &SpawnRate(
    Spawn Probability=&MakeCustomFloatFromBool(Bool=&SetBoolByFloatComparison(A=$System.Age, B=5), True Float=1, False Float=0)
    SpawnRate=&RandomRangeFloat(
     Minimum=5
     Maximum=&Lerp_Float(A=10, B=25, Alpha=$User.NormalizedNotifyProgress)
     Evaluation Type=::NewEnumerator1
    )
    Use Spawn Probability=true
   )
  ]
  particleSpawnScript [
   &self:GetParentInfo(
    Hue Adjustment:NiagaraFloat=&RandomRangeFloat(Minimum=-0.1, Maximum=0.1)
    Normalize Color=true
    Parent Attribute Reader:NiagaraDataInterfaceParticleRead=$Emitter.Particle Attribute Reader
    Saturation Adjustment:NiagaraFloat=&RandomRangeFloat(Minimum=0.5, Maximum=1.2)
    Value Adjustment:NiagaraFloat=&RandomRangeFloat(Minimum=0.3, Maximum=1.5)
   ) {module=GetParentInfo}
   &InitializeParticle(
    Color=$Particles.InheritedColor
    Lifetime=&Lerp_Float(
     A=0.4
     B=2.5
     Alpha=&FloatFromCurve(
      FloatCurve=DataInterface() {propertyValues {Curve {keys [
          {arriveTangent=0.6317, interpMode=RCIM_Cubic, leaveTangent=0.6317, tangentMode=RCTM_User, time=0, value=0}
          {arriveTangent=2.8093, interpMode=RCIM_Cubic, leaveTangent=2.8093, tangentMode=RCTM_User, time=1, value=1}
         ]}}}
      CurveIndex=&RandomRangeFloat(Minimum=0, Maximum=1)
     )
    )
    Position=&AddVectorToPosition(Position=/Niagara/DynamicInputs/Helpers/SimulationPosition(), Vector=<0 0 25>)
    Position Mode=::NewEnumerator1
    Position Offset=&RandomVector(Vector Scale=3)
    Sprite Rotation Mode=::NewEnumerator2
    Sprite Size Mode=::NewEnumerator1
    Uniform Sprite Size Max=20
    Uniform Sprite Size Min=10
    UsePositionOffset=true
    Write Material Random=true
   )
   &AddVelocity(Velocity=&RandomRangeVector(Minimum=<0 0 -5>, Maximum=<0 0 -75>))
  ]
  particleUpdateScript [
   &ParticleState()
   &SpriteRotationRate(Rotation Rate=&RandomRangeFloat(Minimum=-300, Maximum=300))
   &ScaleSpriteSize(
    Scale Sprite Size Mode=::NewEnumerator0
    Uniform Scale Factor=&FloatFromCurve(FloatCurve=DataInterface() {propertyValues {Curve {keys [
         {arriveTangent=3.1235, interpMode=RCIM_Cubic, leaveTangent=3.1235, tangentMode=RCTM_User, time=0, value=0}
         {interpMode=RCIM_Cubic, tangentMode=RCTM_User, time=0.5, value=1}
         {arriveTangent=-3.4019, tangentMode=RCTM_User, time=1, value=0}
        ]}}})
   )
   &Drag(Drag=&RandomRangeFloat(Minimum=1, Maximum=2), Use Linear Drag=true, Use Rotational Drag=false)
   &GravityForce(Gravity=&RandomRangeVector(Minimum=<0 0 -150>, Maximum=<0 0 -25>))
   &SolveForcesAndVelocity(Clamp Velocity=false, Limit Acceleration=false)
   &ScaleColor(Scale Alpha=0.1, ScaleA=true, ScaleRGB=false) {enabled=false}
   /SetVariables(Particles.GlowColor=&MultiplyLinearColor(Linear Color=$Particles.Color, Scale Factor=<0.3 0.3 0.75 1>)) {isSetParameters=true}
  ]
  rendererClasses [&NiagaraSpriteRendererProperties, &NiagaraMeshRendererProperties]
  renderers [
   {rendererClass=&NiagaraSpriteRendererProperties}
   {rendererClass=&NiagaraSpriteRendererProperties, rendererIndex=1}
   {rendererClass=&NiagaraMeshRendererProperties, rendererIndex=2}
  ]
 }
}
summary {
 systemName=NS_Emote_Guitar_SkyLanterns_Solo
 emitters [
  {emitterName=Sparkles, rendererClasses [&NiagaraSpriteRendererProperties, &NiagaraMeshRendererProperties]}
  {
   emitterName=SoloButterfly
   rendererClasses [
    &NiagaraMeshRendererProperties
    &NiagaraSpriteRendererProperties
    &NiagaraLightRendererProperties
   ]
  }
 ]
 userVariables [
  User.bIsFrontEndPreview {type=NiagaraBool}
  User.NormalizedNotifyProgress {type=NiagaraFloat, defaultValue=0, description=""}
 ]
}
